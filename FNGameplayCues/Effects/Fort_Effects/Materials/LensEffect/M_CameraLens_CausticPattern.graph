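# Additive unlit lens-effect material for mesh particles, drawing animated caustics with depth testing disabled.
# Two panning normal maps distort scrolling caustics, sampled at three particle-parameter offsets for RGB fringing, with a single-sample fallback on mobile and low quality.
# The result tints between dark and light colors, scaled by particle color, a soft V-axis mask and time of day.
# CloseUp drives a V-weighted vertex pull toward the object center.
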Material /FNGameplayCues/Effects/Fort_Effects/Materials/LensEffect/M_CameraLens_CausticPattern
graph {
 properties {
  bDisableDepthTest=true
  bUsedWithMeshParticles=true
  bUsedWithNiagaraMeshParticles=true
  blendMode=BLEND_Additive
  pixelDepthOffsetMode=PDOM_Legacy
  shadingModel=MSM_Unlit
  usedShadingModels=Unlit
 }
 nodes [
  n1 = TextureCoordinate()
  n2 = ObjectPositionWS()
  n3 = WorldPosition()
  LinearInterpolate(
   B=Multiply(A=ComponentMask(None=@n1) {g=true}, B=Subtract(A=@n2, B=@n3.XYZ))
   Alpha=DynamicParameter() {defaultValue=<rgba 0 0.3 -0.3 1>, paramNames [CloseUp,OffsetA,OffsetB,Param4]}.CloseUp
  ) {constAlpha=1}
  n4 = VectorParameter() {parameterName=LightColor, defaultValue=<rgba 0.2861 0.5338 0.6 0>}
  n5 = TextureCoordinate() {uTiling=2}
  n6 = Multiply(A=Time(), B=ScalarParameter() {defaultValue=1, parameterName=Panner Speed Scale})
  n7 = Panner(Coordinate=@n5, Time=@n6) {speedX=-0.02, speedY=0.1}
  n8 = Add(
   A=TextureSample(UVs=Panner(Coordinate=@n5, Time=@n6) {speedX=0.02, speedY=0.15}) {samplerType=SAMPLERTYPE_Normal, texture=&T_Weap_Distort_N_02}.RGB
   B=Divide(A=TextureSample(UVs=@n7) {samplerType=SAMPLERTYPE_Normal, texture=/Game/Effects/Fort_Effects/Gameplay/Water/Textures/T_WaterRipples_01}.RGB)
  )
  n9 = Panner(Coordinate=Add(A=TextureCoordinate() {uTiling=10, vTiling=1.2}, B=ComponentMask(None=@n8) {g=true, r=true}), Time=@n6) {speedY=0.07}
  n10 = TextureSample(UVs=@n9) {texture=&T_ShieldCaustics}
  n11 = LinearInterpolate(A=VectorParameter() {parameterName=DarkColor, defaultValue=<rgba 0.0324 0.0686 0.0729 0>}.RGB, B=@n4.RGB, Alpha=@n10.R)
  n12 = DynamicParameter() {defaultValue=<rgba 0 0.3 -0.3 1>, paramNames [CloseUp,OffsetA,OffsetB,Param4]}
  n13 = TextureSample(UVs=Panner(Coordinate=TextureCoordinate() {vTiling=1.5}) {speedX=0.02, speedY=0.15}) {samplerType=SAMPLERTYPE_Normal, texture=&T_Weap_Distort_N_02}
  n14 = Saturate(None=OneMinus(None=MaterialFunctionCall() {materialFunction=&LinearGradient}.VGradient))
  n15 = LinearInterpolate(B=@n13.R, Alpha=@n14)
  n16 = TextureSample(UVs=Add(A=@n9, B=LinearInterpolate(B=@n12.OffsetA, Alpha=@n15) {constB=0.3}) {constB=0.015}) {texture=&T_ShieldCaustics}
  n17 = Add(A=@n9, B=LinearInterpolate(B=@n12.OffsetB, Alpha=LinearInterpolate(B=@n13.G, Alpha=@n14)) {constB=-0.3}) {constB=-0.015}
  n18 = SquareRoot(None=AppendVector(A=AppendVector(A=@n16.R, B=@n10.R), B=TextureSample(UVs=@n17) {texture=&T_ShieldCaustics}.R))
  n19 = Constant() {r=-0.3}
  n20 = QualitySwitch(Default=ShadingPathSwitch(Default=Desaturation(None=@n18, Fraction=@n19), Mobile=@n10.R) {desc=Migrated from FeatureLevelSwitch}, Low=@n10.R)
  n21 = ParticleColor()
  MaterialFunctionCall(Emissive=Multiply(A=LinearInterpolate(A=@n11, B=@n4.RGB, Alpha=@n20), B=@n21.RGB)) {materialFunction=&MF_FX_EmissiveAdjustments}
  n22 = ComponentMask(None=TextureCoordinate()) {g=true}
  n23 = Power(Base=SphereMask(A=OneMinus(None=@n22), B=Constant() {r=0.5}) {attenuationRadius=0.5, hardnessPercent=0})
  n24 = Constant() {r=1.6}
  n25 = Constant() {r=0.5}
  Multiply(
   A=Multiply(A=Saturate(None=@n23), B=@n21.A) {constA=1}
   B=MaterialFunctionCall(Color_Morning=@n24, Color_Day=@n24, Color_Evening=@n25, Color_Night=@n25) {materialFunction=&Color_Time_of_Day}.Result
  )
 ]
}
